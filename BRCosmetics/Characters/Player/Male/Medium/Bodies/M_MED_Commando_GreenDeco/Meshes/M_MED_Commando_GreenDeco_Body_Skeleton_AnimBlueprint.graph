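# Animation Blueprint for the male Commando GreenDeco body cosmetic, following the owner's skeletal mesh and posing its shoulder armor.
# The copied pose feeds the M_MED_Deco_CtrlRig Control Rig, then an additive bone-space move and rotation of dyn_shoulder_armor.
# Each update derives both alphas from bIsSkyDiving: notSkydiving for the rig, isSkydiving for the bone offset.
# An upperarm_l Pose Driver, a Bone Driven Controller and two armor sequence players sit unconnected.
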AnimBlueprint /BRCosmetics/Characters/Player/Male/Medium/Bodies/M_MED_Commando_GreenDeco/Meshes/M_MED_Commando_GreenDeco_Body_Skeleton_AnimBlueprint
target_skeleton=&/M_MED_Commando_GreenDeco_Body_Skeleton
graphs [
 AnimGraph {nodes [
   Bone Driven Controller!(ComponentPose:Component Space Pose Link Structure, Alpha:float=1) -> (Pose:Component Space Pose Link Structure) {is=title AnimGraphNode_BoneDrivenController, alphaBoolBlend {blendOption=Linear}}
   Pose Driver - Source: upperarm_l!(SourcePose:Pose Link Structure) -> (Pose:Pose Link Structure) {
    is=title AnimGraphNode_PoseDriver
    driveOutput=DriveCurves
    poseAsset=&/GreenDeco_ShoulderArmor_PoseAsset
    onlyDriveBones [{boneName=dyn_shoulder_armor}]
    poseTargets [
     {
      bApplyCustomCurve=false
      bIsHidden=false
      distanceMethod=DefaultMethod
      drivenName=Pose_0
      functionType=DefaultFunction
      targetScale=0
      boneTransforms [{targetRotation=<rot -64.7854 -3.089 6.0959>, targetTranslation=<15.5111 0 0>}]
      customCurve {defaultValue=3.4028234663852886e+38, postInfinityExtrap=RCCE_Constant, preInfinityExtrap=RCCE_Constant, keys []}
      targetRotation=<rot 0 0 0>
     }
     {
      bApplyCustomCurve=false
      bIsHidden=false
      distanceMethod=DefaultMethod
      drivenName=Pose_1
      functionType=DefaultFunction
      targetScale=0.0991
      boneTransforms [{targetRotation=<rot -59.8335 -2.7629 5.5534>, targetTranslation=<15.5111 0 0>}]
      customCurve {defaultValue=3.4028234663852886e+38, postInfinityExtrap=RCCE_Constant, preInfinityExtrap=RCCE_Constant, keys []}
      targetRotation=<rot 0 0 0>
     }
     {
      bApplyCustomCurve=false
      bIsHidden=false
      distanceMethod=DefaultMethod
      drivenName=Pose_2
      functionType=DefaultFunction
      targetScale=0.3546
      boneTransforms [{targetRotation=<rot -47.1845 -1.7684 4.1298>, targetTranslation=<15.5111 0 0>}]
      customCurve {defaultValue=3.4028234663852886e+38, postInfinityExtrap=RCCE_Constant, preInfinityExtrap=RCCE_Constant, keys []}
      targetRotation=<rot 0 0 0>
     }
     {
      bApplyCustomCurve=false
      bIsHidden=false
      distanceMethod=DefaultMethod
      drivenName=Pose_3
      functionType=DefaultFunction
      targetScale=0.5519
      boneTransforms [{targetRotation=<rot -30.1493 -0.2694 2.324>, targetTranslation=<15.5111 0 0>}]
      customCurve {defaultValue=3.4028234663852886e+38, postInfinityExtrap=RCCE_Constant, preInfinityExtrap=RCCE_Constant, keys []}
      targetRotation=<rot 0 0 0>
     }
     {
      bApplyCustomCurve=false
      bIsHidden=false
      distanceMethod=DefaultMethod
      drivenName=Pose_4
      functionType=DefaultFunction
      targetScale=0.5155
      boneTransforms [{targetRotation=<rot -12.038 1.4049 0.7911>, targetTranslation=<15.5111 0 0>}]
      customCurve {defaultValue=3.4028234663852886e+38, postInfinityExtrap=RCCE_Constant, preInfinityExtrap=RCCE_Constant, keys []}
      targetRotation=<rot 0 0 0>
     }
     {
      bApplyCustomCurve=false
      bIsHidden=false
      distanceMethod=DefaultMethod
      drivenName=Pose_5
      functionType=DefaultFunction
      targetScale=0.0178
      boneTransforms [{targetRotation=<rot 3.8417 2.894 0.1941>, targetTranslation=<15.5111 0 0>}]
      customCurve {defaultValue=3.4028234663852886e+38, postInfinityExtrap=RCCE_Constant, preInfinityExtrap=RCCE_Constant, keys []}
      targetRotation=<rot 0 0 0>
     }
     {
      bApplyCustomCurve=false
      bIsHidden=false
      distanceMethod=DefaultMethod
      drivenName=Pose_6
      functionType=DefaultFunction
      targetScale=0.5134
      boneTransforms [{targetRotation=<rot 19.6605 3.7971 0.9399>, targetTranslation=<15.5111 0 0>}]
      customCurve {defaultValue=3.4028234663852886e+38, postInfinityExtrap=RCCE_Constant, preInfinityExtrap=RCCE_Constant, keys []}
      targetRotation=<rot 0 0 0>
     }
     {
      bApplyCustomCurve=false
      bIsHidden=false
      distanceMethod=DefaultMethod
      drivenName=Pose_31
      functionType=DefaultFunction
      targetScale=0.5464
      boneTransforms [{targetRotation=<rot 37.6247 4.3399 2.6356>, targetTranslation=<15.5111 0 0>}]
      customCurve {defaultValue=3.4028234663852886e+38, postInfinityExtrap=RCCE_Constant, preInfinityExtrap=RCCE_Constant, keys []}
      targetRotation=<rot 0 0 0>
     }
     {
      bApplyCustomCurve=false
      bIsHidden=false
      distanceMethod=DefaultMethod
      drivenName=Pose_32
      functionType=DefaultFunction
      targetScale=0.4057
      boneTransforms [{targetRotation=<rot 54.4799 5.1458 4.7514>, targetTranslation=<15.5111 0 0>}]
      customCurve {defaultValue=3.4028234663852886e+38, postInfinityExtrap=RCCE_Constant, preInfinityExtrap=RCCE_Constant, keys []}
      targetRotation=<rot 0 0 0>
     }
     {
      bApplyCustomCurve=false
      bIsHidden=false
      distanceMethod=DefaultMethod
      drivenName=Pose_33
      functionType=DefaultFunction
      targetScale=0.1622
      boneTransforms [{targetRotation=<rot 66.9695 6.9552 6.8485>, targetTranslation=<15.5111 0 0>}]
      customCurve {defaultValue=3.4028234663852886e+38, postInfinityExtrap=RCCE_Constant, preInfinityExtrap=RCCE_Constant, keys []}
      targetRotation=<rot 0 0 0>
     }
     {
      bApplyCustomCurve=false
      bIsHidden=false
      distanceMethod=DefaultMethod
      drivenName=Pose_34
      functionType=DefaultFunction
      targetScale=0.1622
      boneTransforms [{targetRotation=<rot 71.831 10.3614 8.342>, targetTranslation=<15.5111 0 0>}]
      customCurve {defaultValue=3.4028234663852886e+38, postInfinityExtrap=RCCE_Constant, preInfinityExtrap=RCCE_Constant, keys []}
      targetRotation=<rot 0 0 0>
     }
     {
      bApplyCustomCurve=false
      bIsHidden=false
      distanceMethod=DefaultMethod
      drivenName=Pose_35
      functionType=DefaultFunction
      targetScale=0.2088
      boneTransforms [{targetRotation=<rot 64.57 21.5796 8.7317>, targetTranslation=<15.5111 0 0>}]
      customCurve {defaultValue=3.4028234663852886e+38, postInfinityExtrap=RCCE_Constant, preInfinityExtrap=RCCE_Constant, keys []}
      targetRotation=<rot 0 0 0>
     }
     {
      bApplyCustomCurve=false
      bIsHidden=false
      distanceMethod=DefaultMethod
      drivenName=Pose_36
      functionType=DefaultFunction
      targetScale=0.6653
      boneTransforms [{targetRotation=<rot 47.0362 41.2455 7.7379>, targetTranslation=<15.5111 0 0>}]
      customCurve {defaultValue=3.4028234663852886e+38, postInfinityExtrap=RCCE_Constant, preInfinityExtrap=RCCE_Constant, keys []}
      targetRotation=<rot 0 0 0>
     }
     {
      bApplyCustomCurve=false
      bIsHidden=false
      distanceMethod=DefaultMethod
      drivenName=Pose_37
      functionType=DefaultFunction
      targetScale=0.7902
      boneTransforms [{targetRotation=<rot 25.7607 62.6974 6.0256>, targetTranslation=<15.5111 0 0>}]
      customCurve {defaultValue=3.4028234663852886e+38, postInfinityExtrap=RCCE_Constant, preInfinityExtrap=RCCE_Constant, keys []}
      targetRotation=<rot 0 0 0>
     }
     {
      bApplyCustomCurve=false
      bIsHidden=false
      distanceMethod=DefaultMethod
      drivenName=Pose_38
      functionType=DefaultFunction
      targetScale=0.3001
      boneTransforms [{targetRotation=<rot 7.7618 79.977 4.4922>, targetTranslation=<15.5111 0 0>}]
      customCurve {defaultValue=3.4028234663852886e+38, postInfinityExtrap=RCCE_Constant, preInfinityExtrap=RCCE_Constant, keys []}
      targetRotation=<rot 0 0 0>
     }
     {
      bApplyCustomCurve=false
      bIsHidden=false
      distanceMethod=DefaultMethod
      drivenName=Pose_39
      functionType=DefaultFunction
      targetScale=0.3362
      boneTransforms [{targetRotation=<rot 0.1949 87.1007 3.8417>, targetTranslation=<15.5111 0 0>}]
      customCurve {defaultValue=3.4028234663852886e+38, postInfinityExtrap=RCCE_Constant, preInfinityExtrap=RCCE_Constant, keys []}
      targetRotation=<rot 0 0 0>
     }
     {
      bApplyCustomCurve=false
      bIsHidden=false
      distanceMethod=DefaultMethod
      drivenName=Pose_40
      functionType=DefaultFunction
      targetScale=0.3001
      boneTransforms [{targetRotation=<rot 1.0923 73.5267 3.6885>, targetTranslation=<15.5111 0 0>}]
      customCurve {defaultValue=3.4028234663852886e+38, postInfinityExtrap=RCCE_Constant, preInfinityExtrap=RCCE_Constant, keys []}
      targetRotation=<rot 0 0 0>
     }
     {
      bApplyCustomCurve=false
      bIsHidden=false
      distanceMethod=DefaultMethod
      drivenName=Pose_41
      functionType=DefaultFunction
      targetScale=1
      boneTransforms [{targetRotation=<rot 2.8997 41.1185 2.5285>, targetTranslation=<15.5111 0 0>}]
      customCurve {defaultValue=3.4028234663852886e+38, postInfinityExtrap=RCCE_Constant, preInfinityExtrap=RCCE_Constant, keys []}
      targetRotation=<rot 0 0 0>
     }
     {
      bApplyCustomCurve=false
      bIsHidden=false
      distanceMethod=DefaultMethod
      drivenName=Pose_42
      functionType=DefaultFunction
      targetScale=0.0178
      boneTransforms [{targetRotation=<rot 3.8434 2.3446 0.1572>, targetTranslation=<15.5111 0 0>}]
      customCurve {defaultValue=3.4028234663852886e+38, postInfinityExtrap=RCCE_Constant, preInfinityExtrap=RCCE_Constant, keys []}
      targetRotation=<rot 0 0 0>
     }
     {
      bApplyCustomCurve=false
      bIsHidden=false
      distanceMethod=DefaultMethod
      drivenName=Pose_43
      functionType=DefaultFunction
      targetScale=0.3827
      boneTransforms [{targetRotation=<rot 3.3273 -30.1538 -1.9312>, targetTranslation=<15.5111 0 0>}]
      customCurve {defaultValue=3.4028234663852886e+38, postInfinityExtrap=RCCE_Constant, preInfinityExtrap=RCCE_Constant, keys []}
      targetRotation=<rot 0 0 0>
     }
     {
      bApplyCustomCurve=false
      bIsHidden=false
      distanceMethod=DefaultMethod
      drivenName=Pose_44
      functionType=DefaultFunction
      targetScale=0.2677
      boneTransforms [{targetRotation=<rot 2.78 -43.7649 -2.6597>, targetTranslation=<15.5111 0 0>}]
      customCurve {defaultValue=3.4028234663852886e+38, postInfinityExtrap=RCCE_Constant, preInfinityExtrap=RCCE_Constant, keys []}
      targetRotation=<rot 0 0 0>
     }
     {
      bApplyCustomCurve=false
      bIsHidden=false
      distanceMethod=DefaultMethod
      drivenName=Pose_45
      functionType=DefaultFunction
      targetScale=0.2677
      boneTransforms [{targetRotation=<rot -4.1646 -39.2893 -6.6884>, targetTranslation=<15.5111 0 0>}]
      customCurve {defaultValue=3.4028234663852886e+38, postInfinityExtrap=RCCE_Constant, preInfinityExtrap=RCCE_Constant, keys []}
      targetRotation=<rot 0 0 0>
     }
     {
      bApplyCustomCurve=false
      bIsHidden=false
      distanceMethod=DefaultMethod
      drivenName=Pose_46
      functionType=DefaultFunction
      targetScale=0.632
      boneTransforms [{targetRotation=<rot -20.798 -28.8122 -16.3633>, targetTranslation=<15.5111 0 0>}]
      customCurve {defaultValue=3.4028234663852886e+38, postInfinityExtrap=RCCE_Constant, preInfinityExtrap=RCCE_Constant, keys []}
      targetRotation=<rot 0 0 0>
     }
     {
      bApplyCustomCurve=false
      bIsHidden=false
      distanceMethod=DefaultMethod
      drivenName=Pose_47
      functionType=DefaultFunction
      targetScale=0.4035
      boneTransforms [{targetRotation=<rot -40.7945 -16.6752 -27.9308>, targetTranslation=<15.5111 0 0>}]
      customCurve {defaultValue=3.4028234663852886e+38, postInfinityExtrap=RCCE_Constant, preInfinityExtrap=RCCE_Constant, keys []}
      targetRotation=<rot 0 0 0>
     }
     {
      bApplyCustomCurve=false
      bIsHidden=false
      distanceMethod=DefaultMethod
      drivenName=Pose_48
      functionType=DefaultFunction
      targetScale=0.0991
      boneTransforms [{targetRotation=<rot -57.6734 -6.9409 -37.4562>, targetTranslation=<15.5111 0 0>}]
      customCurve {defaultValue=3.4028234663852886e+38, postInfinityExtrap=RCCE_Constant, preInfinityExtrap=RCCE_Constant, keys []}
      targetRotation=<rot 0 0 0>
     }
     {
      bApplyCustomCurve=false
      bIsHidden=false
      distanceMethod=DefaultMethod
      drivenName=Pose_49
      functionType=DefaultFunction
      targetScale=0
      boneTransforms [{targetRotation=<rot -64.7854 -3.089 -41.2861>, targetTranslation=<15.5111 0 0>}]
      customCurve {defaultValue=3.4028234663852886e+38, postInfinityExtrap=RCCE_Constant, preInfinityExtrap=RCCE_Constant, keys []}
      targetRotation=<rot 0 0 0>
     }
     {
      bApplyCustomCurve=false
      bIsHidden=false
      distanceMethod=DefaultMethod
      drivenName=Pose_50
      functionType=DefaultFunction
      targetScale=0
      boneTransforms [{targetRotation=<rot -64.7854 -3.089 -32.7874>, targetTranslation=<15.5111 0 0>}]
      customCurve {defaultValue=3.4028234663852886e+38, postInfinityExtrap=RCCE_Constant, preInfinityExtrap=RCCE_Constant, keys []}
      targetRotation=<rot 0 0 0>
     }
     {
      bApplyCustomCurve=false
      bIsHidden=false
      distanceMethod=DefaultMethod
      drivenName=Pose_51
      functionType=DefaultFunction
      targetScale=0
      boneTransforms [{targetRotation=<rot -64.7854 -3.089 -12.5214>, targetTranslation=<15.5111 0 0>}]
      customCurve {defaultValue=3.4028234663852886e+38, postInfinityExtrap=RCCE_Constant, preInfinityExtrap=RCCE_Constant, keys []}
      targetRotation=<rot 0 0 0>
     }
     {
      bApplyCustomCurve=false
      bIsHidden=false
      distanceMethod=DefaultMethod
      drivenName=Pose_52
      functionType=DefaultFunction
      targetScale=0
      boneTransforms [{targetRotation=<rot -64.7854 -3.089 11.667>, targetTranslation=<15.5111 0 0>}]
      customCurve {defaultValue=3.4028234663852886e+38, postInfinityExtrap=RCCE_Constant, preInfinityExtrap=RCCE_Constant, keys []}
      targetRotation=<rot 0 0 0>
     }
     {
      bApplyCustomCurve=false
      bIsHidden=false
      distanceMethod=DefaultMethod
      drivenName=Pose_53
      functionType=DefaultFunction
      targetScale=0
      boneTransforms [{targetRotation=<rot -64.7854 -3.089 31.933>, targetTranslation=<15.5111 0 0>}]
      customCurve {defaultValue=3.4028234663852886e+38, postInfinityExtrap=RCCE_Constant, preInfinityExtrap=RCCE_Constant, keys []}
      targetRotation=<rot 0 0 0>
     }
     {
      bApplyCustomCurve=false
      bIsHidden=false
      distanceMethod=DefaultMethod
      drivenName=Pose_54
      functionType=DefaultFunction
      targetScale=0
      boneTransforms [{targetRotation=<rot -64.7854 -3.089 40.4316>, targetTranslation=<15.5111 0 0>}]
      customCurve {defaultValue=3.4028234663852886e+38, postInfinityExtrap=RCCE_Constant, preInfinityExtrap=RCCE_Constant, keys []}
      targetRotation=<rot 0 0 0>
     }
    ]
    rBFParams {radius=15.4308}
    sourceBones [{boneName=upperarm_l}]
   }
   n1 = Control Rig(
    Source=Copy Pose From Mesh(SourceMeshComponent=Get MeshToCopy() -> (MeshToCopy:Skeletal Mesh Component Object Reference))
    Alpha=Get notSkydiving() -> (notSkydiving:double)
   ) {
    alphaBoolBlend {blendOption=Linear}
    controlRigAssetReference {blueprintRigClass=&/M_MED_Deco_CtrlRig.M_MED_Deco_CtrlRig_C}
   }
   n2 = "Transform (Modify) Bone - Bone: dyn_shoulder_armor"(
    ComponentPose=Local To Component(LocalPose=@n1)
    Alpha=Get isSkydiving() -> (isSkydiving:double)
    Translation="6,0,0"
    Rotation="-16,0,0"
    Scale="0.7,0.7,0.7"
   ) {
    rotationMode=BMM_Additive
    rotationSpace=BCS_BoneSpace
    translationMode=BMM_Additive
    translationSpace=BCS_BoneSpace
    alphaBoolBlend {blendOption=Linear}
    boneToModify {boneName=dyn_shoulder_armor}
    rotation {pitch=-16}
    scale=<0.7 0.7 0.7>
    translation {x=6}
   }
   Output Pose(Result=Component To Local(ComponentPose=@n2)) {layerGroup=None}
   Sequence Player 'M_MED_Commando_GreenDeco_Body_ArmorPoseAnim'() {sequence=&/M_MED_Commando_GreenDeco_Body_ArmorPoseAnim}
   Sequence Player 'M_MED_Commando_GreenDeco_Body_Armor_ROM_Anim'() {sequence=&/M_MED_Commando_GreenDeco_Body_Armor_ROM_Anim}
  ]}
 EventGraph {nodes [
   Event BlueprintUpdateAnimation()
   > Sequence()
   n1 = "To Float (Boolean)"(InBool=Get bIsSkyDiving())
   > Set isSkydiving(isSkydiving:double=@n1) -> (Output_Get:double)
   > Set notSkydiving(notSkydiving:double=float - float(B=@n1) {is=title K2Node_CallFunction}) -> (Output_Get:double)
   Event BlueprintInitializeAnimation()
   > Set MeshToCopy(MeshToCopy:Skeletal Mesh Component Object Reference=GetOwnerSkeletalMeshComponent()) -> (Output_Get:Skeletal Mesh Component Object Reference)
  ]}
]
variables [
 isSkydiving {type=number}
 MeshToCopy {type=SkeletalMeshComponent}
 notSkydiving {type=number}
]
